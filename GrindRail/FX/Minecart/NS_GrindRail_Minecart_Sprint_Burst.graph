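# Niagara effect for the GrindRail minecart sprint burst: a brief flare with spraying glow drops, skid shapes and sparks.
# Foot_Flare001 pairs sprites with a flickering light, while three burst-plus-rate emitters sample skeletal mesh sockets to aim velocity along local -X.
# Skid and spark meshes are pinned to the simulation position, and drops ease away from it under gravity and drag.
# User hue and saturation parameters tint every emitter.
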NiagaraSystem /GrindRail/FX/Minecart/NS_GrindRail_Minecart_Sprint_Burst
dependencies {
 usedDataInterfaces [
  &NiagaraDataInterfaceArrayFloat
  &NiagaraDataInterfaceCurve
  &NiagaraDataInterfaceArrayFloat2
  &NiagaraDataInterfaceMeshRendererInfo
  &NiagaraDataInterfaceArrayFloat3
  &NiagaraDataInterfaceVector2DCurve
  &NiagaraDataInterfaceColorCurve
  &NiagaraDataInterfaceSkeletalMesh
 ]
 usedDynamicInputs [
  &SimulationPosition
  &Vector2DFromFloat
  &SelectFloatFromArray
  &AddVectorToPosition
 ]
 usedModules [
  &self:SystemSpawnScript.SystemScriptSource.SystemScriptGraph.NiagaraNodeAssignment_2.SetVariables_#1
  &SystemState
 ]
}
emitters {
 Foot_Flare001 {
  emitterUpdateScript [
   &EmitterState(
    Inactive Response=::NewEnumerator1
    Life Cycle Mode=::NewEnumerator1
    Loop Behavior=::NewEnumerator1
    Loop Duration Mode=::NewEnumerator1
    UseLoopDelay=false
   )
   &SpawnBurst_Instantaneous(Spawn Count=1, Use Spawn Probability=false)
  ]
  particleSpawnScript [
   &InitializeParticle(
    Color=&HueShiftLinearColor(Hue=$User.SprintColor_Hue, Saturation=$User.SprintColor_Saturation, Value=1, Linear Color=<rgba 0.9375 0.375 0.0938 1>)
    Lifetime=0.25
    Mass Mode=::NewEnumerator1
    Sprite Rotation Mode=::NewEnumerator2
    Sprite Size Mode=::NewEnumerator3
    Uniform Sprite Size=75
    UsePositionOffset=false
    Write Material Random=true
   ) {module=InitializeParticle001}
   &CameraOffset(Camera Offset Amount=25)
  ]
  particleUpdateScript [
   &ParticleState()
   &SpriteRotationRate(Rotation Rate=10)
   &DynamicMaterialParameters(
    Index 0 Param 1=&FloatFromCurve(FloatCurve=DataInterface() {propertyValues {Curve {keys [<at 0 1 RCIM_Cubic>,<at 0.8 0 RCIM_Cubic>]}}})
    Index 0 Param 2=2
    Index 0 Param 3=1
    Index 0 Param 4=1
    Index 1 Param 1=&FloatFromCurve(FloatCurve=DataInterface() {propertyValues {Curve {keys [
         {arriveTangent=-2.3163, interpMode=RCIM_Cubic, time=0, value=2}
         {arriveTangent=-1.9019, interpMode=RCIM_Cubic, leaveTangent=-1.9019, tangentMode=RCTM_User, time=1, value=0.4}
        ]}}})
    Index 1 Param 2=0.4
    Index 1 Param 3=0.6
    Index 1 Param 4=1
    Index 2 Param 1=&FloatFromCurve(FloatCurve=DataInterface() {propertyValues {Curve {keys [<at 0.46 0>,<at 1 1>]}}})
    Index 2 Param 2=45
    Module.Param0WriteEnabled=true
    Module.Param0WriteEnabled1=false
    Module.Param0WriteEnabled2=true
    Module.Param1WriteEnabled=true
    Module.Param1WriteEnabled1=true
    Module.Param1WriteEnabled2=true
    Module.Param2WriteEnabled=true
    Module.Param2WriteEnabled1=true
    Module.Param2WriteEnabled2=true
    Module.Param3WriteEnabled=true
    Module.Param3WriteEnabled1=true
    Module.Param3WriteEnabled2=true
    Write Parameter Index 1=true
    Write Parameter Index 2=true
   )
   &ScaleColor(
    Scale RGB=&VectorFromFloat(Value=&FloatFromCurve(
      FloatCurve=DataInterface() {propertyValues {Curve {keys [
          {interpMode=RCIM_Cubic, tangentMode=RCTM_User, time=0, value=1}
          {interpMode=RCIM_Cubic, tangentMode=RCTM_User, time=0.5, value=1.5}
          {tangentMode=RCTM_User, time=1, value=1}
         ]}}}
      Scale Curve=0.25
     ))
    ScaleA=false
    ScaleRGB=true
   ) {module=ScaleColor002}
   &FN_Light_Attributes(
    Color=&MultiplyLinearColor(
     Linear Color=&HueShiftLinearColor(Hue=$User.SprintColor_Hue, Saturation=$User.SprintColor_Saturation, Value=1, Linear Color=<rgba 1 0.2684 0 1>)
     Scale Factor=&MakeFloat4FromVectorAndFloat("Vector (RGB)"=<3 3 3>, "Float (Alpha)"=1)
    )
    Create Color Attribute=true
    Create Exponent Attribute=true
    Create Radius Attribute=true
    Radius=40
    Scale Color=&Multiply_Float(
     A=1
     B=&Add_Float(A=0.5, B=&Waveform("Waveform [1]"=::NewEnumerator9, "[1] Frequency"=20, "[1] Amplitude Min/Max"=<0.5 1>))
    )
   )
   &FN_Sprite_Attributes(
    Color=&Multiply_LinearColorByFloat(
     Linear Color=&HueShiftLinearColor(Hue=$User.SprintColor_Hue, Saturation=$User.SprintColor_Saturation, Value=1, Linear Color=<rgba 7 2.8 1.4 1>)
     Float=1
    )
    Create Camera Offset Attribute=true
    Create Color Attribute=true
    Create Sprite Rotation Attribute=true
    Create SpriteSize Attribute=true
    SpriteSize=&Multiply_Vector2D_ByFloat(
     Float=&FloatFromCurve(FloatCurve=DataInterface() {propertyValues {Curve {keys [
          {arriveTangent=9.0725, interpMode=RCIM_Cubic, leaveTangent=9.0725, tangentMode=RCTM_User, time=0, value=0}
          {arriveTangent=-2.6365, interpMode=RCIM_Cubic, leaveTangent=-2.6365, tangentMode=RCTM_User, time=0.1408, value=0.9847}
          {arriveTangent=-0.0405, interpMode=RCIM_Cubic, leaveTangent=-0.0405, tangentMode=RCTM_User, time=1, value=0}
         ]}}})
     Vector2D=<150 150>
    )
    Write to Module Namespace=true
   )
   &FN_Sprite_Attributes(
    Camera Offset=20
    Color=&Multiply_LinearColorByFloat(
     Linear Color=&HueShiftLinearColor(Hue=$User.SprintColor_Hue, Saturation=$User.SprintColor_Saturation, Value=1, Linear Color=<rgba 0.1094 0.0404 0.0059 1>)
     Float=1
    )
    Create Camera Offset Attribute=true
    Create Color Attribute=true
    Create Sprite Rotation Attribute=true
    Create SpriteSize Attribute=true
    SpriteSize=&Multiply_Vector2D_ByFloat(
     Float=&FloatFromCurve(FloatCurve=DataInterface() {propertyValues {Curve {keys [
          {arriveTangent=9.0725, interpMode=RCIM_Cubic, leaveTangent=9.0725, tangentMode=RCTM_User, time=0, value=0}
          {arriveTangent=-0.2277, interpMode=RCIM_Cubic, leaveTangent=-0.2277, tangentMode=RCTM_User, time=0.1408, value=0.9847}
          {arriveTangent=-0.0405, interpMode=RCIM_Cubic, leaveTangent=-0.0405, tangentMode=RCTM_User, time=1, value=0}
         ]}}})
     Vector2D=<50 50>
    )
    Write to Module Namespace=true
   ) {module=FN_Sprite_Attributes001}
   /SetVariables(Particles.DMP_Flare:Vector4f=&MakeVector4(X=1, Y=1, Z=0, W=0)) {isSetParameters=true}
  ]
  rendererClasses [&NiagaraLightRendererProperties, &NiagaraSpriteRendererProperties]
  renderers [
   {rendererClass=&NiagaraLightRendererProperties, rendererIndex=0}
   {rendererClass=&NiagaraSpriteRendererProperties, rendererIndex=1}
   {rendererClass=&NiagaraSpriteRendererProperties, rendererIndex=2}
  ]
 }
 GlowDrops_CPU_Local {
  emitterUpdateScript [
   &EmitterState(
    Enable Distance Culling=true
    Life Cycle Mode=::NewEnumerator1
    Loop Behavior=::NewEnumerator1
    Loop Duration Mode=::NewEnumerator1
    Max Distance Response=::KillImmediately
    Module.bApplySpawnCountScaleByDistance=true
    Module.bUseMaxDistance=true
    Module.bUseMinDistance=false
    Scalability Mode=::NewEnumerator1
    Scale Spawn Count=true
    Spawn Count Scale=3
    Spawn Count Scale By Distance=DataInterface() {propertyValues {Curve {keys [<at 0.3676 0.9657>,<at 0.9962 0.4982>]}}}
    UseLoopDelay=true
   )
   &SpawnRate(SpawnRate=30, Use Spawn Probability=false)
   &SpawnBurst_Instantaneous(Spawn Count=10, Spawn Group=1, Use Spawn Probability=false, UseLoopCountLimit=false)
  ]
  particleSpawnScript [
   /SetVariables(Particles.Rand:NiagaraInt32=&RandomRangeInt(Minimum=0, Maximum=1)) {isSetParameters=true}
   &InitializeParticle(
    Color Mode=::NewEnumerator0
    Lifetime Max=0.2
    Lifetime Min=0.1
    Lifetime Mode=::NewEnumerator1
    Sprite Size Max=&SelectVector2DFromArray(Direct Array Index=$Particles.Rand, Vector 2D Selection Array=DataInterface() {propertyValues {FloatData [<4 4>, <2 10>]}})
    Sprite Size Min=&SelectVector2DFromArray(Direct Array Index=$Particles.Rand, Vector 2D Selection Array=DataInterface() {propertyValues {FloatData [<2 2>, <1 5>]}})
    Sprite Size Mode=::NewEnumerator2
    UsePositionOffset=false
    Write Material Random=true
   )
   &ShapeLocation(Box Size=<2 10 2>, MaskSpawnGroup=false, Override Seed=false, Shape Primitive=::NewEnumerator2, Surface Only=false, UseOwnerScale=true)
   &SkeletalMeshLocation(
    MaskSpawnGroup=false
    Mesh Sampling Type=::NewEnumerator5
    Orientation Sampling=::NewEnumerator1
    Position Sampling=::NewEnumerator2
    Sampled Orientation Coordinate Space=::World
    Skeletal Mesh=$System.Skeletal Mesh
    Socket Sampling Mode=::NewEnumerator5
   )
   &AddVelocity(
    Additional Quaternion Rotation=false
    Cone Angle=&RandomRangeFloat(Minimum=20, Maximum=40)
    Cone Axis=<-1 0 0>
    Inner Cone Angle=10
    Module.UseConeFalloff=false
    Module.UseVelDistribution=false
    Override Seed=false
    Rotation Mode=::NewEnumerator3
    Rotation Quaternion:Quat4f=$Particles.SkeletalMeshLocation.SampledRotation
    Velocity Mode=::NewEnumerator2
    Velocity Speed=&RandomRangeFloat(Minimum=500, Maximum=750)
   )
  ]
  particleUpdateScript [
   &ParticleState()
   &GravityForce(Gravity=&RandomRangeVector(Minimum=<0 0 -920>, Maximum=<0 0 -480>))
   &GravityForce(Coordinate Space=::Local, Gravity=&RandomRangeVector(Minimum=<0 0 0>, Maximum=<-750 0 0>)) {module=GravityForce001}
   &AerodynamicDrag(
    Aerodynamic Drag:NiagaraFloat=&RandomRangeFloat(Minimum=0.2, Maximum=1)
    Aerodynamic Rotational Drag=0.2
    Surface Normal Axis:Vector3f=&SelectVectorFromArray(Direct Array Index=$Particles.SpawnGroup, Vector Selection Array=DataInterface() {propertyValues {FloatData [<0.2 -0.2 1>, <0.2 -0.2 1>]}})
    Use Linear Drag=true
    Use Rotational Drag=true
   )
   &ScaleVelocity(Velocity Scale=&RandomRangeVector(Minimum=<0 0 0>, Maximum=<1 1 1>))
   &InheritVelocity(
    Inherited Velocity Amount Scale=&RandomRangeVector(Minimum=<0 0 0>, Maximum=<0.25 0.25 0.25>)
    Speed Limit=true
    UseSpeedThreshold=false
   )
   &SolveForcesAndVelocity(Clamp Velocity=false, Limit Acceleration=false)
   /SetVariables(Particles.Position=&Lerp_Vector(
     A=Linked() {linkedVariable {name=Particles.Position, type=NiagaraPosition}}
     B=/Niagara/DynamicInputs/Helpers/SimulationPosition()
     Alpha=&FloatFromCurve(FloatCurve=DataInterface() {propertyValues {Curve {keys [
          {arriveTangent=-3.0707, interpMode=RCIM_Cubic, leaveTangent=-3.0707, tangentMode=RCTM_User, time=0, value=1}
          {arriveTangent=0.1636, interpMode=RCIM_Cubic, leaveTangent=0.1636, tangentMode=RCTM_User, time=1, value=0}
         ]}}})
    )) {isSetParameters=true}
   &ScaleSpriteSize(Scale Factor=&Vector2DFromCurve(
     Vector2Curve=DataInterface() {propertyValues {
       XCurve {keys [<at 0 1 RCIM_Cubic>, {arriveTangent=-1.2073, interpMode=RCIM_Cubic, leaveTangent=-1.2073, tangentMode=RCTM_User, time=1, value=0.55}]}
       YCurve {keys [<at 0 1 RCIM_Cubic>, {arriveTangent=-1.2073, interpMode=RCIM_Cubic, leaveTangent=-1.2073, tangentMode=RCTM_User, time=1, value=0.55}]}
      }}
     ScaleCurve=<1 1>
    ))
   &ScaleSpriteSizeBySpeed(Max Scale Factor=<0.5 5>, Min Scale Factor=<0.75 1.5>)
   /SetVariables(Particles.Color=&MultiplyLinearColor(
     Linear Color=&HueShiftLinearColor(
      Hue=$User.SprintColor_Hue
      Saturation=$User.SprintColor_Saturation
      Value=1
      Linear Color=&LerpLinearColors(
       StartColor=<rgba 25 6.25 0 1>
       EndColor=<rgba 25 5.2083 0 1>
       LerpFactor=&Multiply_Float(A=$Particles.NormalizedAge, B=1.5)
      )
     )
     Scale Factor=&MakeVector4_FromFloat(Float=1)
    )) {isSetParameters=true}
   &AlignSpriteToMeshOrientation(SetAlignment=true, SetFacing=true)
  ]
  rendererClasses [&NiagaraSpriteRendererProperties]
  renderers [{rendererClass=&NiagaraSpriteRendererProperties}]
 }
 SkidSharp {
  emitterUpdateScript [
   &EmitterState(Life Cycle Mode=::NewEnumerator1, Loop Behavior=::NewEnumerator1, Loop Duration Mode=::NewEnumerator1, UseLoopDelay=false)
   &SpawnRate(SpawnRate=30, Use Spawn Probability=false)
   &SpawnBurst_Instantaneous(Spawn Count=10, Use Spawn Probability=false, UseLoopCountLimit=false)
  ]
  particleSpawnScript [
   &InitializeParticle(
    Color=&HueShiftLinearColor(Hue=$User.SprintColor_Hue, Saturation=$User.SprintColor_Saturation, Value=1, Linear Color=<rgba 5 1.5 0 1>)
    Lifetime Max=0.15
    Lifetime Min=0.1
    Lifetime Mode=::NewEnumerator1
    Mesh Scale=<0.6 0.35 0.2>
    Mesh Scale Mode=::NewEnumerator4
    Position Offset=<5 0 0>
    UsePositionOffset=true
    Write Material Random=true
   ) {module=InitializeParticle001}
   &ShapeLocation(
    Box Size=<0 2 0>
    MaskSpawnGroup=false
    Non Uniform Scale=<0 0 0>
    Override Seed=false
    Shape Primitive=::NewEnumerator2
    Surface Only=false
    UseOwnerScale=true
   )
   &DynamicMaterialParameters(
    Index 0 Param 3=&RandomRangeFloat(Minimum=1, Maximum=1.5, Evaluation Type=::NewEnumerator1)
    Module.Param0WriteEnabled=false
    Module.Param1WriteEnabled=false
    Module.Param2WriteEnabled=true
    Module.Param3WriteEnabled=false
   ) {module=DynamicMaterialParameters001}
   &SkeletalMeshLocation(
    MaskSpawnGroup=false
    Mesh Sampling Type=::NewEnumerator5
    Orientation Sampling=::NewEnumerator1
    Position Sampling=::NewEnumerator2
    Sampled Orientation Coordinate Space=::World
    Skeletal Mesh=$System.Skeletal Mesh
    Socket Sampling Mode=::NewEnumerator5
   )
   &AddVelocity(
    Additional Quaternion Rotation=false
    Cone Angle=10
    Cone Axis=<-1 0 0>
    Module.UseConeFalloff=false
    Module.UseVelDistribution=false
    Override Seed=false
    Rotation Mode=::NewEnumerator3
    Rotation Quaternion:Quat4f=$Particles.SkeletalMeshLocation.SampledRotation
    Velocity Mode=::NewEnumerator2
    Velocity Speed=&RandomRangeFloat(Minimum=150, Maximum=300)
   )
  ]
  particleUpdateScript [
   &ParticleState()
   &ScaleColor(
    Scale Alpha=&FloatFromCurve(FloatCurve=DataInterface() {propertyValues {Curve {keys [
         {arriveTangent=-0.2241, interpMode=RCIM_Cubic, leaveTangent=-0.2241, tangentMode=RCTM_User, time=0, value=1}
         {arriveTangent=-3.1123, interpMode=RCIM_Cubic, leaveTangent=-3.1123, tangentMode=RCTM_User, time=1, value=0}
        ]}}})
    Scale RGB=&MakeVectorFromLinearColorRGB(Color=&ColorFromCurve(ColorCurve=DataInterface() {propertyValues {
        AlphaCurve {keys [<at 0 1 RCIM_Cubic>]}
        BlueCurve {keys [<at 0.0137 1>,<at 0.9732 0.75>]}
        GreenCurve {keys [<at 0.0137 1>,<at 0.9732 0.75>]}
        RedCurve {keys [<at 0.0137 1>,<at 0.9732 0.75>]}
       }}))
    ScaleA=true
    ScaleRGB=true
   )
   /SetVariables(Particles.Position=&SimulationPosition()) {isSetParameters=true}
   /SetVariables(Particles.Renderer2_Color:LinearColor=&HueShiftLinearColor(
     Hue=$User.SprintColor_Hue
     Saturation=$User.SprintColor_Saturation
     Value=1
     Linear Color=&ColorFromCurve(ColorCurve=DataInterface() {propertyValues {
        AlphaCurve {keys [<at 0.1971 1 RCIM_Cubic>,<at 0.9879 0>]}
        BlueCurve {keys [<at 0.0137 0>,<at 0.4297 0>,<at 0.9484 0>]}
        GreenCurve {keys [<at 0.0137 1.6667>,<at 0.4297 0.2649>,<at 0.9484 0.2649>]}
        RedCurve {keys [<at 0.0137 5>,<at 0.4297 1>,<at 0.9484 1>]}
       }})
    )) {isSetParameters=true}
   &DynamicMaterialParameters(
    Index 0 Param 1=&FloatFromCurve(FloatCurve=DataInterface() {propertyValues {Curve {keys [<at 0.1 0.1 RCIM_Cubic>,<at 0.8 1>]}}})
    Index 0 Param 2=2
    Index 0 Param 4=&FloatFromCurve(FloatCurve=DataInterface() {propertyValues {Curve {keys [<at 0 1>,<at 1 3>]}}})
    Module.Param0WriteEnabled=true
    Module.Param1WriteEnabled=true
    Module.Param2WriteEnabled=false
    Module.Param3WriteEnabled=true
   )
   &ScaleMeshSize(Scale Factor=&VectorFromFloat(Value=&FloatFromCurve(
      FloatCurve=DataInterface() {propertyValues {Curve {keys [
          {arriveTangent=2.4702, interpMode=RCIM_Cubic, leaveTangent=2.4702, tangentMode=RCTM_User, time=0, value=0}
          {arriveTangent=0.9995, interpMode=RCIM_Cubic, leaveTangent=0.9995, tangentMode=RCTM_User, time=0.4, value=0.75}
          {arriveTangent=-0.0405, interpMode=RCIM_Cubic, leaveTangent=-0.0405, tangentMode=RCTM_User, time=1, value=1}
         ]}}}
      Scale Curve=0.55
     ))) {module=ScaleMeshSize001}
   /SetVariables(Particles.Renderer2_Size:Vector3f=&Multiply_Vector(A=$Particles.Scale, B=<2.1 2.1 2.1>)) {isSetParameters=true}
   /SetVariables(Particles.Renderer2_CamOffset=-5) {isSetParameters=true}
   &SolveForcesAndVelocity(Clamp Velocity=false, Limit Acceleration=false)
  ]
  rendererClasses [&NiagaraMeshRendererProperties]
  renderers [
   {rendererClass=&NiagaraMeshRendererProperties, rendererIndex=0}
   {rendererClass=&NiagaraMeshRendererProperties, rendererIndex=1}
  ]
 }
 Spark_Splines_LOW {
  emitterUpdateScript [
   &EmitterState(Life Cycle Mode=::NewEnumerator1, Loop Behavior=::NewEnumerator1, Loop Duration Mode=::NewEnumerator1, UseLoopDelay=false)
   &SpawnRate(SpawnRate=30, Use Spawn Probability=false)
   &SpawnBurst_Instantaneous(Spawn Count=10, Spawn Group=1, Use Spawn Probability=false, UseLoopCountLimit=false)
  ]
  particleSpawnScript [
   &InitializeParticle(
    Color=&HueShiftLinearColor(Hue=$User.SprintColor_Hue, Saturation=$User.SprintColor_Saturation, Value=1, Linear Color=<rgba 5 1.2865 0.2083 1>)
    Lifetime=0.25
    Mesh Scale Max=&MergeHelper_VectorSwitch(0=<0.3 0.3 0.5>, 1=<0.6 0.6 0.1>, 2=?, 3=?, Index=&SpawnGroup(), Num=::NewEnumerator0)
    Mesh Scale Min=&MergeHelper_VectorSwitch(0=<0.1 0.1 0.2>, 1=<0.25 0.25 0.05>, 2=?, 3=?, Index=&SpawnGroup(), Num=::NewEnumerator0)
    Mesh Scale Mode=::NewEnumerator2
    Position Offset=<5 0 0>
    UsePositionOffset=true
    Write Material Random=true
   )
   &ShapeLocation(
    Box Size=<0 2 0>
    MaskSpawnGroup=false
    Non Uniform Scale=<0 0 0>
    Override Seed=false
    Shape Primitive=::NewEnumerator2
    Surface Only=false
    UseOwnerScale=true
   )
   &SkeletalMeshLocation(
    MaskSpawnGroup=false
    Mesh Sampling Type=::NewEnumerator5
    Orientation Sampling=::NewEnumerator1
    Position Sampling=::NewEnumerator2
    Skeletal Mesh=$System.Skeletal Mesh
    Socket Sampling Mode=::NewEnumerator5
   )
   &AddVelocity(
    Additional Quaternion Rotation=false
    Cone Angle=10
    Cone Axis=<-1 0 0>
    Module.UseConeFalloff=false
    Module.UseVelDistribution=false
    Override Seed=false
    Rotation Mode=::NewEnumerator3
    Rotation Quaternion:Quat4f=$Particles.SkeletalMeshLocation.SampledRotation
    Velocity Mode=::NewEnumerator2
    Velocity Speed=&RandomRangeFloat(Minimum=150, Maximum=300)
   )
   &InitialMeshOrientation(
    Mesh Orientation Mode=::NewEnumerator2
    Rotation=&MakeVector(X=&RandomRangeFloat(Minimum=0, Maximum=1), Y=0, Z=0)
    Use Rotation Vector=true
   ) {module=InitialMeshOrientation001}
  ]
  particleUpdateScript [
   &ParticleState()
   /SetVariables(Particles.Position=&SimulationPosition()) {isSetParameters=true}
   &ScaleMeshSize(Scale Factor=&VectorFromFloat(Value=&Multiply_Float(A=4, B=1)))
   &ScaleColor(
    Scale Alpha=&FloatFromCurve(FloatCurve=DataInterface() {propertyValues {Curve {keys [
         {arriveTangent=18.7405, interpMode=RCIM_Cubic, leaveTangent=18.7405, tangentMode=RCTM_User, time=0, value=0}
         {interpMode=RCIM_Cubic, tangentMode=RCTM_User, time=0.1, value=1}
         {interpMode=RCIM_Cubic, tangentMode=RCTM_User, time=0.5, value=1}
         {interpMode=RCIM_Cubic, tangentMode=RCTM_User, time=1, value=0}
        ]}}})
    ScaleA=true
    ScaleRGB=true
   )
   &DynamicMaterialParameters(
    Index 0 Param 1=&FloatFromCurve(FloatCurve=DataInterface() {propertyValues {Curve {keys [
         {arriveTangent=11.6708, interpMode=RCIM_Cubic, leaveTangent=11.6708, tangentMode=RCTM_User, time=0, value=1}
         {arriveTangent=0.5512, interpMode=RCIM_Cubic, leaveTangent=0.5512, tangentMode=RCTM_User, time=1, value=15}
        ]}}})
    Index 0 Param 2=&FloatFromCurve(FloatCurve=DataInterface() {propertyValues {Curve {keys [
         {arriveTangent=-0.0433, interpMode=RCIM_Cubic, leaveTangent=-0.0433, tangentMode=RCTM_User, time=0, value=-1}
         {arriveTangent=4.1024, interpMode=RCIM_Cubic, leaveTangent=4.1024, tangentMode=RCTM_User, time=1, value=1}
        ]}}})
    Index 0 Param 3=&RandomRangeFloat(Minimum=0.5, Maximum=2)
    Module.Index0Param0Write=true
    Module.Index0Param1Write=true
    Module.Index0Param2Write=true
    Module.Index0Param3Write=false
   )
   &SolveForcesAndVelocity(Clamp Velocity=false, Limit Acceleration=false)
  ]
  rendererClasses [&NiagaraMeshRendererProperties]
  renderers [{rendererClass=&NiagaraMeshRendererProperties, rendererIndex=0}]
 }
}
summary {
 systemName=NS_GrindRail_Minecart_Sprint_Burst
 emitters [
  {emitterName=Foot_Flare001, rendererClasses [&NiagaraLightRendererProperties, &NiagaraSpriteRendererProperties]}
  {emitterName=GlowDrops_CPU_Local, rendererClasses [&NiagaraSpriteRendererProperties]}
  {emitterName=SkidSharp, rendererClasses [&NiagaraMeshRendererProperties]}
  {emitterName=Spark_Splines_LOW, rendererClasses [&NiagaraMeshRendererProperties]}
 ]
 userVariables [
  User.SprintColor_Hue {type=NiagaraFloat, defaultValue=0, description=""}
  User.SprintColor_Saturation {type=NiagaraFloat, defaultValue=1, description=""}
 ]
}
